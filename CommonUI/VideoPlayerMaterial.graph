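# VideoPlayerMaterial is an unlit UI-domain material in CommonUI that, by its name, serves as the surface for video playback in widgets.
# Its graph holds a single 2D texture sample parameter, MediaTexture, which defaults to DefaultMediaTexture.
# The parameter lets a material instance substitute another texture, with no other processing nodes present.
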Material /CommonUI/VideoPlayerMaterial
graph {
 properties {materialDomain=MD_UI, pixelDepthOffsetMode=PDOM_Legacy, shadingModel=MSM_Unlit, usedShadingModels=Unlit}
 nodes [TextureSampleParameter2D() {parameterName=MediaTexture, texture=&/DefaultMediaTexture}]
}
