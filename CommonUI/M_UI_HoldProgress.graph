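# Translucent UI material for a hold-progress ring on a pointer, filled radially by the HoldTime parameter.
# An SDF circle function draws the outline, with PointerSize intended to scale radius and stroke.
# A polar angle gradient forms an arc mask for a SmoothStep cutoff at HoldTime, its tiny soft edge applied only above zero.
# Max layers the masked progress arc over what appears to be the half-strength outline.
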Material /CommonUI/M_UI_HoldProgress
graph {
 properties {
  bAutomaticallySetUsageInEditor=false
  bUsedWithStaticMesh=false
  blendMode=BLEND_Translucent
  materialDomain=MD_UI
  usedShadingModels=Default Lit
 }
 nodes [
  n1 = ScalarParameter() {defaultValue=0.3, group=Pointer, parameterName=HoldTime, sliderMax=1}
  n2 = Constant()
  n3 = If(A=@n1, "A > B"=Constant() {r=0.002}, "A == B"=@n2, "A < B"=@n2)
  n4 = OneMinus(None=SmoothStep(Min=@n1, Max=Add(A=@n1, B=@n3), Value=NamedRerouteUsage()))
  n5 = MaterialFunctionCall!(
   UV
   Size=Multiply(A=Constant() {r=0.6}, B=NamedRerouteUsage())
   GlowMin=Constant() {r=0.01}
   GlowMax=Constant() {r=0.02}
   Stroke thickness=Multiply(A=Constant() {r=0.04}, B=NamedRerouteUsage())
   Is stroke thickness relative to size
  ) {materialFunction=/CommonUI/MaterialFunctions/MF_UI_SDF_Circle}
  Max(
   A=Multiply(A=NamedRerouteUsage()) {constB=0.5}
   B=NamedRerouteDeclaration(None=Min(A=@n4, B=@n5.Outline)) {name=HoldOutlineProgress, nodeColor=<rgba 1 0.1875 0 1>}
  )
  n6 = MaterialFunctionCall!(Vector or UVs=TextureCoordinate(), Swizzle Coordinate Output) {materialFunction=/CommonUI/MaterialFunctions/MF_UI_PolarCoordinates}
  n7 = Constant() {r=1}
  n8 = Constant() {r=1}
  NamedRerouteDeclaration(None=MaterialFunctionCall!(
    In=OneMinus(None=@n6.Angle gradient)
    Value 1=Constant()
    Value 2=@n7
    Value 3=@n7
    Value 4=@n7
    Value 5=@n7
    Pos value 1=Constant()
    Pos value 2=@n8
    Pos value 3=@n8
    Pos value 4=@n8
    Pos value 5=@n8
   ) {materialFunction=/CommonUI/MaterialFunctions/MF_UI_RemappableGradient}.Gradient 3 Colors) {name=ArcMask, nodeColor=<rgba 0 1 0.7422 1>}
  NamedRerouteDeclaration(None=@n5.Outline) {name=HoldOutline, nodeColor=<rgba 0 0.1328 1 1>}
  NamedRerouteDeclaration(None=ScalarParameter() {defaultValue=1, group=Pointer, parameterName=PointerSize, sliderMax=1.25, sliderMin=0.75}) {name=PointerSizeScalar, nodeColor=<rgba 0 0 0 1>}
 ]
}
