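# Translucent UI material that draws a procedural virtual pointer from SDF circle functions: center dot, stroked ring and soft glow.
# A radial-segments function masks the ring outline into dashes, with the IsDragable parameter apparently blending between solid and dashed styles.
# PointerSize provides a scale control for the pointer.
# Named reroutes combine the layers into FinalColor and FinalOpacity, with opacity adding an extra faint term.
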Material /CommonUI/M_UI_VirtualPointer
graph {
 properties {
  bAutomaticallySetUsageInEditor=false
  bUsedWithStaticMesh=false
  blendMode=BLEND_Translucent
  materialDomain=MD_UI
  usedShadingModels=Default Lit
 }
 nodes [
  n1 = MaterialFunctionCall!(
   UV
   Size=Multiply(A=Constant() {r=0.5}, B=NamedRerouteUsage())
   GlowMin=Constant() {r=0.01}
   GlowMax=Constant() {r=0.03}
   Stroke thickness=Multiply(A=Constant(), B=NamedRerouteUsage())
   Is stroke thickness relative to size
  ) {materialFunction=/CommonUI/MaterialFunctions/MF_UI_SDF_Circle}
  NamedRerouteDeclaration(None=@n1.Fill) {name=StrokeFill, nodeColor=<rgba 0 0.5078 1 1>}
  NamedRerouteDeclaration(None=MaterialFunctionCall(Size=Constant() {r=0.01}, GlowMin=Constant() {r=0.01}, GlowMax=Constant() {r=0.03}, Stroke thickness=Constant()) {materialFunction=&MF_UI_SDF_Circle}.Fill) {name=DotFill, nodeColor=<rgba 1 0 0.9141 1>}
  n2 = Constant() {r=0.55}
  n3 = NamedRerouteUsage()
  NamedRerouteDeclaration(None=MaterialFunctionCall(Size=Multiply(A=@n2, B=@n3), GlowMin=Constant() {r=0.01}, GlowMax=Constant() {r=0.15}, Stroke thickness=Constant()) {materialFunction=&MF_UI_SDF_Circle}.Fill) {name=GlowFill, nodeColor=<rgba 0 1 0.7656 1>}
  NamedRerouteDeclaration(None=ScalarParameter() {defaultValue=1, group=Pointer, parameterName=PointerSize, sliderMax=1.25, sliderMin=0.75}) {name=PointerSizeScalar, nodeColor=<rgba 0 0 0 1>}
  n4 = Multiply(A=Constant() {r=22}, B=NamedRerouteUsage())
  n5 = Constant() {r=0.5}
  NamedRerouteDeclaration(None=Min(
    A=NamedRerouteDeclaration(None=@n1.Outline) {name=StrokeOutline, nodeColor=<rgba 0.7578 1 0 1>}
    B=MaterialFunctionCall!(UV, Segment Count=@n4, Segment Spacing=@n5, Fixed Spacing) {materialFunction=/CommonUI/MaterialFunctions/MF_UI_RadialSegments}.Segmented Triangles Fill
   )) {name=DashFill, nodeColor=<rgba 0 1 0.9063 1>}
  n6 = NamedRerouteUsage()
  n7 = NamedRerouteUsage()
  n8 = Add(
   A=Multiply(A=Subtract(A=@n6, B=@n7)) {constB=0.3}
   B=LinearInterpolate(A=NamedRerouteUsage(), B=NamedRerouteUsage(), Alpha=NamedRerouteUsage())
  )
  NamedRerouteDeclaration(None=Add(A=@n8, B=NamedRerouteUsage())) {name=FinalColor, nodeColor=<rgba 0 1 0.3906 1>}
  NamedRerouteDeclaration(None=ScalarParameter() {group=Pointer, parameterName=IsDragable, sliderMax=1}) {name=PointerDragStyleScalar, nodeColor=<rgba 0 0 0 1>}
  n9 = Multiply(A=NamedRerouteUsage(), B=Constant() {r=0.1}) {constB=0.1}
  NamedRerouteDeclaration(None=Add(A=Add(A=@n8, B=@n9), B=NamedRerouteUsage())) {name=FinalOpacity, nodeColor=<rgba 1 0.0703 0 1>}
 ]
}
