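# CommonUI virtual cursor widget that shows pointer state through two material-driven images, a pointer and a hold-progress indicator.
# Construction creates their dynamic materials and applies the default look.
# State changes reset visuals, then test ActiveStates bitmasks to set material parameters for hover and press sizing or drag.
# Hold progress is unwired: its event is empty, and the state handler only passes zero hold time, collapsing the indicator.
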WidgetBlueprint /CommonUI/WBP_VirtualPointer
graphs [
 EventGraph {nodes [
   n1 = Event PreConstruct()
   n2 = GetDynamicMaterial(execute=[@n1.then, @n4.then], self=Get Image_Pointer() -> (Image_Pointer:Image Object Reference))
   > Set MID_Pointer(MID_Pointer:Material Instance Dynamic Object Reference=@n2) -> (Output_Get:Material Instance Dynamic Object Reference)
   > n3 = GetDynamicMaterial(self=Get Image_HoldProgress() -> (Image_HoldProgress:Image Object Reference))
   > Set MID_HoldProgress(MID_HoldProgress:Material Instance Dynamic Object Reference=@n3) -> (Output_Get:Material Instance Dynamic Object Reference)
   > SetDefaultVisualState()
   n4 = Event Construct()
   n5 = Event OnVirtualPointerVisualStateChanged!() -> (OutputDelegate, then, PreviousStates:ECursorVisualState Enum=Default, ActiveStates:ECursorVisualState Enum=Default) {is=title K2Node_Event}
   > SetDefaultVisualState()
   n6 = Make Bitmask!(Bitmask:Bitmask Integer=1) -> (ReturnValue:Bitmask Integer) {is=title K2Node_BitmaskLiteral}
   n7 = "To Integer (Byte)"!(InByte:Byte=@n5.ActiveStates) -> (ReturnValue:int=0)
   > n8 = Branch(Condition="Not Equal (Integer)"(
     A=Bitwise AND(A:int=@n7, B:int=@n6) -> (ReturnValue:int=0) {is=title K2Node_CommutativeAssociativeBinaryOperator}
     B=MakeLiteralInt!(Value:int=0) -> (ReturnValue:int=0)
    ))
   > n9 = SetHoverVisualState()
   n10 = Make Bitmask(Bitmask=2)
   Branch(
    execute=[@n9.then, @n8.else]
    Condition="Not Equal (Integer)"(A=Bitwise AND(A=@n7, B=@n10) -> (ReturnValue=0), B=MakeLiteralInt(Value=0) -> (ReturnValue=0))
   )
   > SetPressedVisualState()
   n11 = Make Bitmask(Bitmask=4)
   n12 = Branch(Condition="Not Equal (Integer)"(A=Bitwise AND(A=@n7, B=@n11) -> (ReturnValue=0), B=MakeLiteralInt(Value=0) -> (ReturnValue=0)))
   > n13 = SetDragVisualState()
   n14 = Make Bitmask(Bitmask=4)
   Branch(
    execute=[@n13.then, @n12.else]
    Condition="Equal (Integer)"(A=Bitwise AND(A=@n7, B=@n14) -> (ReturnValue=0), B=MakeLiteralInt(Value=0) -> (ReturnValue=0))
   )
   > SetHoldVisualState!(execute, self, HoldTime:double=0) -> (then)
   SetHoldVisualState(HoldTime=0)
   Event Tick()
   Event OnVirtualPointerHoldProgress!() -> (OutputDelegate, then, Progress:float) {is=title K2Node_Event}
  ]}
 SetDefaultVisualState {nodes [
   SetDefaultVisualState() {is=title K2Node_FunctionEntry}
   > SetScalarParameterValue(self=Get MID_Pointer() -> (MID_Pointer:Material Instance Dynamic Object Reference), ParameterName=PointerSize, Value=1)
   > SetScalarParameterValue(self=Get MID_Pointer(), ParameterName=IsDragable, Value=0)
   > SetScalarParameterValue(self=Get MID_HoldProgress() -> (MID_HoldProgress:Material Instance Dynamic Object Reference), ParameterName=HoldTime, Value=0)
   > SetVisibility(self=Get Image_HoldProgress(), InVisibility=Collapsed)
   > SetScalarParameterValue(self=Get MID_HoldProgress(), ParameterName=PointerSize, Value=1)
  ]}
 SetDragVisualState {nodes [
   SetDragVisualState() {is=title K2Node_FunctionEntry}
   > SetScalarParameterValue(self=Get MID_Pointer(), ParameterName=IsDragable, Value=1)
  ]}
 SetHoldVisualState {nodes [
   n1 = SetHoldVisualState!() -> (then, HoldTime:double=~) {is=title K2Node_FunctionEntry}
   > SetVisibility(self=Get Image_HoldProgress(), InVisibility=Select(Option 0=Collapsed, Option 1=SelfHitTestInvisible, Index="float > float"(A=@n1)))
   > SetScalarParameterValue(self=Get MID_HoldProgress(), ParameterName=HoldTime, Value=Get HoldTime() -> (HoldTime:double))
  ]}
 SetHoverVisualState {nodes [
   SetHoverVisualState() {is=title K2Node_FunctionEntry}
   > SetScalarParameterValue(self=Get MID_Pointer(), ParameterName=PointerSize, Value=1.25)
   > SetScalarParameterValue(self=Get MID_HoldProgress(), ParameterName=PointerSize, Value=1.25)
  ]}
 SetPressedVisualState {nodes [
   SetPressedVisualState() {is=title K2Node_FunctionEntry}
   > SetScalarParameterValue(self=Get MID_Pointer(), ParameterName=PointerSize, Value=0.75)
   > SetScalarParameterValue(self=Get MID_HoldProgress(), ParameterName=PointerSize, Value=0.75)
  ]}
]
variables [
 MID_HoldProgress {type=MaterialInstanceDynamic}
 MID_Pointer {type=MaterialInstanceDynamic}
]
widgets {
 info {inheritedWidgetCount=0, parentClass=&CommonUserWidget, rootWidgetClass=&GridPanel, widgetCount=3}
 widgets [
  GridPanel_56 {widgetClassPath=&GridPanel, widget=GridPanel_56, properties {columnFill [0], rowFill [0]}}
  Image_Pointer {
   widgetClassPath=&Image
   bIsVariable=true
   parent=GridPanel_56
   widget=Image_Pointer
   properties {brush {
     bIsDynamicallyLoaded=false
     drawAs=Image
     imageType=FullColor
     mirroring=NoMirror
     resourceName=None
     resourceObject=&/M_UI_VirtualPointer
     tiling=NoTile
     imageSize=<120 120>
     margin=<ltrb 0 0 0 0>
     outlineSettings {
      bUseBrushTransparency=false
      roundingType=HalfHeightRadius
      width=0
      color {colorUseRule=UseColor_Specified, specifiedColor=<rgba 0 0 0 0>}
      cornerRadii=<0 0 0 0>
     }
     tintColor {colorUseRule=UseColor_Specified, specifiedColor=<rgba 1 1 1 1>}
     uVRegion {bIsValid=false, max=<0 0>, min=<0 0>}
    }}
   slot {class=GridSlot, properties {horizontalAlignment=HAlign_Center, verticalAlignment=VAlign_Center, padding=<ltrb -60 -60 -60 -60>}}
  }
  Image_HoldProgress {
   widgetClassPath=&Image
   bIsVariable=true
   parent=GridPanel_56
   widget=Image_HoldProgress
   properties {brush {
     bIsDynamicallyLoaded=false
     drawAs=Image
     imageType=FullColor
     mirroring=NoMirror
     resourceName=None
     resourceObject=&/M_UI_HoldProgress
     tiling=NoTile
     imageSize=<120 120>
     margin=<ltrb 0 0 0 0>
     outlineSettings {
      bUseBrushTransparency=false
      roundingType=HalfHeightRadius
      width=0
      color {colorUseRule=UseColor_Specified, specifiedColor=<rgba 0 0 0 0>}
      cornerRadii=<0 0 0 0>
     }
     tintColor {colorUseRule=UseColor_Specified, specifiedColor=<rgba 1 1 1 1>}
     uVRegion {bIsValid=false, max=<0 0>, min=<0 0>}
    }}
   slot {class=GridSlot, properties {padding=<ltrb -60 -60 -60 -60>}}
  }
 ]
}
